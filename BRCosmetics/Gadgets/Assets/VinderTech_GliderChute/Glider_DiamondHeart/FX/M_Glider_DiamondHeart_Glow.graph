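# Translucent, two-sided unlit glow material for the Diamond Heart glider's Niagara sprite and mesh particle effects.
# Particle color RGB passes through MF_FX_EmissiveAdjustments for emissive.
# A side glow mask, blended between squared and raw values by SrgbTickedOn and shaped by an exponent, multiplies particle alpha.
# That alpha is softened by a smoothstepped camera-to-normal angle fade and a depth fade, which is bypassed on mobile.
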Material /BRCosmetics/Gadgets/Assets/VinderTech_GliderChute/Glider_DiamondHeart/FX/M_Glider_DiamondHeart_Glow
graph {
 properties {
  bUsedWithNiagaraMeshParticles=true
  bUsedWithNiagaraSprites=true
  blendMode=BLEND_Translucent
  pixelDepthOffsetMode=PDOM_Legacy
  shadingModel=MSM_Unlit
  twoSided=true
  usedShadingModels=Unlit
 }
 nodes [
  n1 = ParticleColor()
  MaterialFunctionCall(Emissive=@n1.RGB) {materialFunction=&MF_FX_EmissiveAdjustments}
  n2 = TextureSampleParameter2D() {
   parameterName=Glow Texture
   samplerType=SAMPLERTYPE_LinearColor
   texture=&/T_DiamondHeart_Glider_SideGlowMask
  }
  n3 = Power(
   Base=LinearInterpolate(A=Multiply(A=@n2.R, B=@n2.R), B=@n2.R, Alpha=ScalarParameter() {parameterName=SrgbTickedOn})
   Exp=ScalarParameter() {defaultValue=1, parameterName=GlowExponent}
  )
  n4 = ScalarParameter() {parameterName=View Angle Fade Min, sliderMax=1}
  n5 = ScalarParameter() {defaultValue=1, parameterName=View Angle Fade Max, sliderMax=1}
  n6 = Abs(None=DotProduct(A=CameraVectorWS(), B=VertexNormalWS()))
  n7 = DepthFade(FadeDistance=ScalarParameter() {defaultValue=50, parameterName=DepthFade})
  Multiply(
   A=Multiply(A=Multiply(A=@n1.A, B=@n3), B=SmoothStep(Min=@n4, Max=@n5, Value=@n6))
   B=ShadingPathSwitch(Default=@n7, Deferred=@n7, Mobile=Constant() {r=1}) {desc=Migrated from FeatureLevelSwitch}
  )
 ]
}
