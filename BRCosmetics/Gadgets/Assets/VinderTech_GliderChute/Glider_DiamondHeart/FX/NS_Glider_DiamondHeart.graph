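# Niagara system for the Diamond Heart glider cosmetic, adding glints, glow and trails attached to its skeletal mesh.
# A socket-following mesh emitter pulses glint alpha by sine, while a multi-renderer sprite emitter layers differently oriented glow cards.
# Soft, blue-tip and gold-tip ribbon emitters use paired bone-driven AnimTrails that fade in on start.
# In gameplay, system Timeline time drives trail opacity, with front-end overrides and time-of-day color adjustments.
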NiagaraSystem /BRCosmetics/Gadgets/Assets/VinderTech_GliderChute/Glider_DiamondHeart/FX/NS_Glider_DiamondHeart
dependencies {
 usedDataInterfaces [&NiagaraDataInterfaceCurve, &NiagaraDataInterfaceColorCurve, &NiagaraDataInterfaceCamera]
 usedDynamicInputs [&SetBoolByFloatComparison, &Add_Float, &Power]
 usedModules [
  &SystemState
  &Timeline
  &self:SystemSpawnScript.SystemScriptSource.SystemScriptGraph.NiagaraNodeAssignment_1.SetVariables_#1
  &self:SystemSpawnScript.SystemScriptSource.SystemScriptGraph.NiagaraNodeAssignment_0.SetVariables_#2
 ]
}
emitters {
 DiamondGlints {
  emitterUpdateScript [
   &EmitterState(Inactive Response=::NewEnumerator1, Life Cycle Mode=::NewEnumerator1, Loop Behavior=::NewEnumerator1, Loop Duration=1, UseLoopDelay=false)
   &SpawnBurst_Instantaneous(Use Spawn Probability=false, UseLoopCountLimit=false)
   &SpawnRate(SpawnRate=0.75, Use Spawn Probability=false) {enabled=false}
  ]
  particleSpawnScript [
   &InitializeParticle(Lifetime=1.5, Sprite Size=<70 70>, Sprite Size Mode=::NewEnumerator4, UsePositionOffset=true, Write Material Random=true)
   &CameraOffset()
  ]
  particleUpdateScript [
   &SkeletalMeshLocation(
    Bone / Socket Index=4
    MaskSpawnGroup=false
    Mesh Sampling Type=::NewEnumerator5
    Sampled Position Offset=<0 2 -7>
    Skeletal Mesh=$System.Skeletal Mesh
    Socket Sampling Mode=::NewEnumerator5
   )
   &ParticleState(Kill Particles When Lifetime Has Elapsed=false)
   &ScaleColor(Scale Alpha=&Sine(Normalized Angle=$Emitter.Age, Period=6, Scale=1, Bias=2), ScaleA=true, ScaleRGB=true)
   &DynamicMaterialParameters(
    Index 0 Param 1=1
    Index 0 Param 2=0.5
    Index 0 Param 3=1
    Index 0 Param 4=1
    Index 1 Param 2=30
    Module.Index0Param0Write=true
    Module.Index0Param1Write=true
    Module.Index0Param2Write=true
    Module.Index0Param3Write=true
    Module.Index1Param0Write=true
    Module.Index1Param1Write=true
    Module.Index1Param2Write=true
    Module.Index1Param3Write=true
    Write Parameter Index 1=true
   )
  ]
  rendererClasses [&NiagaraMeshRendererProperties]
  renderers [{rendererClass=&NiagaraMeshRendererProperties, rendererIndex=0}]
 }
 Glow {
  emitterUpdateScript [
   &EmitterState(
    Inactive Response=::NewEnumerator1
    Life Cycle Mode=::NewEnumerator1
    Loop Behavior=::NewEnumerator1
    Loop Duration=8.049
    UseLoopDelay=false
   )
   &SpawnBurst_Instantaneous(Use Spawn Probability=false, UseLoopCountLimit=false)
   &SpawnRate(SpawnRate=0.75, Use Spawn Probability=false) {enabled=false}
   &SpawnBurst_Instantaneous(Spawn Time=8.0489, Use Spawn Probability=false, UseLoopCountLimit=false) {module=SpawnBurst_Instantaneous001}
  ]
  particleSpawnScript [
   &InitializeParticle(
    Color=<rgba 0.4892 0.6662 1 0.5>
    Lifetime=1.5
    Position Offset=<0 -10 0>
    Sprite Size=<70 70>
    Sprite Size Mode=::NewEnumerator4
    UsePositionOffset=true
    Write Material Random=true
   )
   &CameraOffset(Camera Offset Amount=4)
  ]
  particleUpdateScript [
   &SkeletalMeshLocation(
    Bone / Socket Index=4
    MaskSpawnGroup=false
    Mesh Sampling Type=::NewEnumerator5
    Skeletal Mesh=$System.Skeletal Mesh
    Socket Sampling Mode=::NewEnumerator5
   )
   &SpriteFacingAndAlignment(Sprite Facing=<0 0 1>, WriteAlignment=true, WriteFacing=true)
   /SetVariables(Particles.SpriteSize=<55 110>) {isSetParameters=true}
   /SetVariables(Particles.Color=&MultiplyLinearColor(
     Linear Color=$Particles.Initial.Color
     Scale Factor=&MakeFloat4FromVectorAndFloat(
      "Vector (RGB)"=<1 1 1>
      "Float (Alpha)"=&FloatFromCurve(
       FloatCurve=DataInterface() {propertyValues {Curve {keys [<at 0.1 1>,<at 0.4 1>]}}}
       CurveIndex=&Abs_Float(Float:NiagaraFloat=&DotProduct(Vector A=&ReturnCameraVector(Camera Vectors=::NewEnumerator1, Coordinate Space=::Simulation), Vector B=<0 0 1>))
      )
     )
    )) {isSetParameters=true}
   /SetVariables(
    Particles.ColorFront:LinearColor=&MultiplyLinearColor(
     Linear Color=$Particles.Initial.Color
     Scale Factor=&MakeFloat4FromVectorAndFloat("Vector (RGB)"=<1 1 1>, "Float (Alpha)"=0.5)
    )
    Particles.ColorFrontBot:LinearColor=&MultiplyLinearColor(
     Linear Color=$Particles.Initial.Color
     Scale Factor=&MakeFloat4FromVectorAndFloat("Vector (RGB)"=<1 1 1>, "Float (Alpha)"=1)
    )
    Particles.PositionFront:NiagaraPosition=&AddVectorToPosition(Position=$Particles.Position, Vector=<0 -5 0>)
    Particles.PositionFrontBot:NiagaraPosition=&AddVectorToPosition(Position=$Particles.Position, Vector=<0 -20 0>)
    Particles.SpriteAlignmentFront:Vector3f=<1 0 0>
    Particles.SpriteAlignmentFrontBot:Vector3f=<-1 0 0>
    Particles.SpriteFacingFront:Vector3f=<0 1 0>
    Particles.SpriteSizeFront:Vector2f=<70 65>
    Particles.SpriteSizeFrontBot:Vector2f=<70 45>
   ) {isSetParameters=true}
   /SetVariables(
    Particles.ColorSide:LinearColor=&MultiplyLinearColor(
     Linear Color=$Particles.Initial.Color
     Scale Factor=&MakeFloat4FromVectorAndFloat("Vector (RGB)"=<1 1 1>, "Float (Alpha)"=0.45)
    )
    Particles.SpiteSizeSide:Vector2f=<70 130>
    Particles.SpriteAlignmentSide:Vector3f=<0 1 0>
    Particles.SpriteFacingSide:Vector3f=<1 0 0>
   ) {isSetParameters=true}
   /SetVariables(Particles.SpriteFacingCross1:Vector3f=<1 1 0>, Particles.SpriteFacingCross2:Vector3f=<1 -1 0>) {isSetParameters=true}
   &ParticleState(Kill Particles When Lifetime Has Elapsed=false)
   &DynamicMaterialParameters(
    Index 1 Param 2=30
    Module.Index0Param0Write=true
    Module.Index0Param1Write=true
    Module.Index0Param2Write=true
    Module.Index0Param3Write=true
    Module.Index1Param0Write=true
    Module.Index1Param1Write=true
    Module.Index1Param2Write=true
    Module.Index1Param3Write=true
    Write Parameter Index 1=true
   )
  ]
  rendererClasses [&NiagaraSpriteRendererProperties]
  renderers [
   {rendererClass=&NiagaraSpriteRendererProperties}
   {rendererClass=&NiagaraSpriteRendererProperties, rendererIndex=1}
   {rendererClass=&NiagaraSpriteRendererProperties, rendererIndex=2}
   {rendererClass=&NiagaraSpriteRendererProperties, rendererIndex=3}
   {rendererClass=&NiagaraSpriteRendererProperties, rendererIndex=4}
   {rendererClass=&NiagaraSpriteRendererProperties, rendererIndex=5}
  ]
 }
 NE_SoftTrail {
  emitterSpawnScript [&FN_RibbonIDOffsetUtility()]
  emitterUpdateScript [
   &EmitterState()
   &FN_SpawnAnimTrail(Number of Anim Trails=2)
  ]
  particleSpawnScript [
   &FN_InitializeAnimTrailBoneIndecies(First Bone or Socket Index=1, Second Bone or Socket Index=5)
   &FN_InitializeAnimTrailBoneIndecies(AnimTrail Selection=1, First Bone or Socket Index=3, Second Bone or Socket Index=6) {module=FN_InitializeAnimTrailBoneIndecies001}
   /SetVariables(Particles.AlphaFadeIn:NiagaraFloat=&FloatFromCurve(FloatCurve=DataInterface() {propertyValues {Curve {keys [<at 0 0>,<at 0.5 1>]}}}, CurveIndex=$Emitter.Age)) {isSetParameters=true}
   &FN_InitializeAnimTrail(
    AnimTrailWidth=&FN_FrontEndFloatModifier(Gameplay=1, Front End=0.6)
    Color=&MakeLinearColor(
     R:NiagaraFloat=1
     G:NiagaraFloat=1
     B:NiagaraFloat=1
     A:NiagaraFloat=&Multiply_Float(A=$Particles.AlphaFadeIn, B=1)
    )
    Lifetime=&MakeCustomFloatFromBool(Bool=$Particles.Locked, True Float=100, False Float=0.4)
    Position Offset=&FN_FrontEndVector3Modifier(
     Gameplay=&MakeVector(X=&MaskFloatBySpawnGroup(Masked Float=2, Passthrough Float=-2), Y=-40, Z=0)
     Front End=&MakeVector(X=&MaskFloatBySpawnGroup(Masked Float=2, Passthrough Float=-2), Y=-40, Z=0)
    )
    SkeletalMesh=$System.Skeletal Mesh
   )
   &AddVelocity(Velocity=&Multiply_VectorByFloat(Vector=<0 800 0>, Float=&MakeCustomFloatFromBool(Bool=$User.bIsFrontEnd, True Float=1, False Float=1)))
  ]
  particleUpdateScript [
   &SolveForcesAndVelocity(Clamp Velocity=false, Limit Acceleration=false)
   &FN_UpdateAnimTrail(
    Locked Particle Color:LinearColor=&MultiplyLinearColor(
     Linear Color=$Particles.Initial.Color
     Scale Factor=&MakeFloat4FromVectorAndFloat("Vector (RGB)"=<1 1 1>, "Float (Alpha)"=$System.Timeline.Time)
    )
    Locked Particle RibbonWidth=1
    SkeletalMesh=$System.Skeletal Mesh
   ) {module=FN_UpdateAnimTrail001}
   &Color(
    Color=&MultiplyLinearColor(
     Linear Color=<rgba 1 1 1 0.25>
     Scale Factor=&MakeVector4_FromLinearColor(LinearColor=&FNTimeOfDayColor(Night Color=<rgba 0.2 0.2 0.2 1>, Morning Color=<rgba 1 1 1 1>, Day Color=<rgba 1 1 1 1>, Evening Color=<rgba 1 1 1 1>))
    )
    Scale Alpha=&Multiply_Float(
     A=&Multiply_Float(A=0.5, B=$Particles.AlphaFadeIn)
     B=&FN_FrontEndFloatModifier(Gameplay=$System.Timeline.Time, Front End=1)
    )
   )
  ]
  rendererClasses [&NiagaraRibbonRendererProperties]
  renderers [
   {rendererClass=&NiagaraRibbonRendererProperties}
   {rendererClass=&NiagaraRibbonRendererProperties, rendererIndex=1}
  ]
 }
 NE_TipBlueTrail {
  emitterSpawnScript [&FN_RibbonIDOffsetUtility()]
  emitterUpdateScript [
   &EmitterState()
   &FN_SpawnAnimTrail(Number of Anim Trails=2)
  ]
  particleSpawnScript [
   &FN_InitializeAnimTrailBoneIndecies()
   &FN_InitializeAnimTrailBoneIndecies(AnimTrail Selection=1, First Bone or Socket Index=2, Second Bone or Socket Index=3) {module=FN_InitializeAnimTrailBoneIndecies001}
   /SetVariables(Particles.AlphaFadeIn:NiagaraFloat=&FloatFromCurve(FloatCurve=DataInterface() {propertyValues {Curve {keys [<at 0 0>,<at 0.5 1>]}}}, CurveIndex=$Emitter.Age)) {isSetParameters=true, module=SetVariables_#3}
   &FN_InitializeAnimTrail(
    AnimTrailWidth=&FN_FrontEndFloatModifier(Gameplay=1, Front End=0.7)
    Color=&MakeLinearColor(
     R:NiagaraFloat=1
     G:NiagaraFloat=1
     B:NiagaraFloat=1
     A:NiagaraFloat=&Multiply_Float(A=$Particles.AlphaFadeIn, B=0.5)
    )
    Lifetime=&MakeCustomFloatFromBool(Bool=$Particles.Locked, True Float=100, False Float=0.4)
    Position Offset=&FN_FrontEndVector3Modifier(Gameplay=<0 0 0>, Front End=<0 0 0>)
    SkeletalMesh=$System.Skeletal Mesh
   )
   &AddVelocity(Velocity=&Multiply_VectorByFloat(Vector=<0 800 0>, Float=&MakeCustomFloatFromBool(Bool=$User.bIsFrontEnd, True Float=1, False Float=1)))
   &DynamicMaterialParameters(
    Index 0 Param 1=1
    Module.Param0WriteEnabled=true
    Module.Param1WriteEnabled=true
    Module.Param2WriteEnabled=true
    Module.Param3WriteEnabled=true
   )
   /SetVariables() {enabled=false, isSetParameters=true, module=SetVariables_#4}
   &CameraOffset()
  ]
  particleUpdateScript [
   &SolveForcesAndVelocity(Clamp Velocity=false, Limit Acceleration=false)
   &FN_UpdateAnimTrail(
    Locked Particle Color:LinearColor=&MultiplyLinearColor(
     Linear Color=$Particles.Initial.Color
     Scale Factor=&MakeFloat4FromVectorAndFloat("Vector (RGB)"=<1 1 1>, "Float (Alpha)"=$System.Timeline.Time)
    )
    Locked Particle RibbonWidth=1
    SkeletalMesh=$System.Skeletal Mesh
   ) {module=FN_UpdateAnimTrail001}
   /SetVariables(Particles.RibbonU0Override=$Particles.NormalizedAge) {isSetParameters=true, module=SetVariables_#5}
   &Color(
    Color=&MultiplyLinearColor(
     Linear Color=&ColorFromCurve(ColorCurve=DataInterface() {propertyValues {
        AlphaCurve {keys [<at 0 1 RCIM_Cubic>]}
        BlueCurve {keys [<at 0.003 0.4219>,<at 1 0.05 RCIM_Cubic>]}
        GreenCurve {keys [<at 0.003 0.3219>,<at 1 0.05 RCIM_Cubic>]}
        RedCurve {keys [<at 0.003 0.2021>,<at 1 0.05 RCIM_Cubic>]}
       }})
     Scale Factor=&MakeVector4_FromLinearColor(LinearColor=&FNTimeOfDayColor(Night Color=<rgba 3 3 3 1>, Morning Color=<rgba 1 1 1 1>, Day Color=<rgba 1 1 1 1>, Evening Color=<rgba 1 1 1 1>))
    )
    Scale Alpha=&Multiply_Float(
     A=&Multiply_Float(A=0.5, B=$Particles.AlphaFadeIn)
     B=&FN_FrontEndFloatModifier(Gameplay=$System.Timeline.Time, Front End=1)
    )
    Scale Color=&VectorFromFloat(Value=&FNTimeOfDayScalar(Night Scale=0.15, Morning Scale=1, Evening Scale=0.5))
   )
   &ScaleColor(ScaleA=true, ScaleRGB=true) {enabled=false}
  ]
  rendererClasses [&NiagaraRibbonRendererProperties]
  renderers [{rendererClass=&NiagaraRibbonRendererProperties}]
 }
 NE_TipGoldTrail {
  emitterSpawnScript [&FN_RibbonIDOffsetUtility()]
  emitterUpdateScript [
   &EmitterState()
   &FN_SpawnAnimTrail(AnimTrailWidthMode=::FromSecond, Number of Anim Trails=2)
  ]
  particleSpawnScript [
   &FN_InitializeAnimTrailBoneIndecies()
   &FN_InitializeAnimTrailBoneIndecies(AnimTrail Selection=1, First Bone or Socket Index=2, Second Bone or Socket Index=3) {module=FN_InitializeAnimTrailBoneIndecies001}
   /SetVariables(Particles.AlphaFadeIn:NiagaraFloat=&FloatFromCurve(FloatCurve=DataInterface() {propertyValues {Curve {keys [<at 0 0>,<at 0.5 1>]}}}, CurveIndex=$Emitter.Age)) {isSetParameters=true, module=SetVariables_#3}
   &FN_InitializeAnimTrail(
    AnimTrailWidth=0.1
    Color=&MakeLinearColor(
     R:NiagaraFloat=1
     G:NiagaraFloat=1
     B:NiagaraFloat=1
     A:NiagaraFloat=&Multiply_Float(A=$Particles.AlphaFadeIn, B=0.5)
    )
    Lifetime=&MakeCustomFloatFromBool(Bool=$Particles.Locked, True Float=100, False Float=0.4)
    Position Offset=&FN_FrontEndVector3Modifier(Gameplay=<0 0 0>, Front End=<0 0 0>)
    SkeletalMesh=$System.Skeletal Mesh
   )
   &AddVelocity(Velocity=&Multiply_VectorByFloat(Vector=<0 800 0>, Float=&MakeCustomFloatFromBool(Bool=$User.bIsFrontEnd, True Float=1, False Float=1)))
   &DynamicMaterialParameters(
    Index 0 Param 1=1
    Module.Param0WriteEnabled=true
    Module.Param1WriteEnabled=true
    Module.Param2WriteEnabled=true
    Module.Param3WriteEnabled=true
   )
   /SetVariables() {enabled=false, isSetParameters=true, module=SetVariables_#4}
   &CameraOffset()
  ]
  particleUpdateScript [
   &SolveForcesAndVelocity(Clamp Velocity=false, Limit Acceleration=false)
   &FN_UpdateAnimTrail(
    Locked Particle Color:LinearColor=&MultiplyLinearColor(
     Linear Color=$Particles.Initial.Color
     Scale Factor=&MakeFloat4FromVectorAndFloat("Vector (RGB)"=<1 1 1>, "Float (Alpha)"=$System.Timeline.Time)
    )
    Locked Particle RibbonWidth=1
    SkeletalMesh=$System.Skeletal Mesh
   ) {module=FN_UpdateAnimTrail001}
   /SetVariables(Particles.RibbonU0Override=$Particles.NormalizedAge) {isSetParameters=true, module=SetVariables_#5}
   &Color(
    Color=&MultiplyLinearColor(
     Linear Color=&ColorFromCurve(ColorCurve=DataInterface() {propertyValues {
        AlphaCurve {keys [<at 0 1 RCIM_Cubic>]}
        BlueCurve {keys [<at 0.003 0.2365>,<at 0.104 0.2359>]}
        GreenCurve {keys [<at 0.003 1.3759>,<at 0.104 0.5014>]}
        RedCurve {keys [<at 0.003 2>,<at 0.104 0.651>]}
       }})
     Scale Factor=&MakeVector4_FromLinearColor(LinearColor=&FNTimeOfDayColor(Night Color=<rgba 3 3 3 1>, Morning Color=<rgba 1 1 1 1>, Day Color=<rgba 1 1 1 1>, Evening Color=<rgba 1 1 1 1>))
    )
    Scale Alpha=&Multiply_Float(
     A=&Multiply_Float(A=0.5, B=$Particles.AlphaFadeIn)
     B=&FN_FrontEndFloatModifier(Gameplay=$System.Timeline.Time, Front End=1)
    )
    Scale Color=&VectorFromFloat(Value=&FNTimeOfDayScalar(Night Scale=0.15, Morning Scale=1, Evening Scale=0.5))
   )
   &ScaleColor(ScaleA=true, ScaleRGB=true) {enabled=false}
  ]
  rendererClasses [&NiagaraRibbonRendererProperties]
  renderers [{rendererClass=&NiagaraRibbonRendererProperties}]
 }
}
summary {
 systemName=NS_Glider_DiamondHeart
 emitters [
  {emitterName=NE_SoftTrail, rendererClasses [&NiagaraRibbonRendererProperties]}
  {emitterName=NE_TipBlueTrail, rendererClasses [&NiagaraRibbonRendererProperties]}
  {emitterName=Glow, rendererClasses [&NiagaraSpriteRendererProperties]}
  {emitterName=DiamondGlints, rendererClasses [&NiagaraMeshRendererProperties]}
  {emitterName=NE_TipGoldTrail, rendererClasses [&NiagaraRibbonRendererProperties]}
 ]
 userVariables [
  User.bIsFrontEnd {type=NiagaraBool}
  User.bIsFrontEndPreview {type=NiagaraBool}
  User.bIsFullyDeployed {type=NiagaraBool}
  User.bisSwinging {type=NiagaraBool}
  User.ForwardDot {type=NiagaraFloat, defaultValue=1, description=""}
  User.Offset {type=Vector3f, description="", defaultValue {struct=Vector3f, value=<0 0 0>}}
  User.RibbonIDOffset {type=NiagaraInt32}
  User.RightDot {type=NiagaraFloat, defaultValue=0, description=""}
  User.Size {type=NiagaraFloat, defaultValue=1, description=""}
  User.SkeletalMesh {
   type=NiagaraDataInterfaceSkeletalMesh
   defaultValue {
    struct=NiagaraExt_VariableValue_DataInterface
    value {dataInterface=&self:NiagaraDataInterfaceSkeletalMesh_0}
   }
  }
  User.VariantPicker {type=NiagaraFloat, defaultValue=0, description=""}
 ]
}
